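; Music party radio Blueprint that starts and stops streamed media audio with the music party.
; Start and stop events set a party-active flag, then call StartRadio or StopRadio, which check state flags and pass a debounce check.
; Starting binds playback-resumed to a timeline that fades in MediaSound volume; a failed play calls Stop.
; ChangeStation is empty, and the stop event's FadeOutDuration is unused.
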(event EventBeginPlay
  (Variables|FortMediaStreamingRadio|SetSoundComponent (Variables|Default|GetMediaSound))
  (EventDispatchers|BindEventtoOnPlayFailed (delegate Custom|On%20Play%20Failed)))

(event Custom|On%20Play%20Failed
  (CallFunction|Stop))

(event EventStartMusicParty
  (Variables|Default|SetbMusicPartyIsActive true)
  (Default|StartRadio))

(event EventStopMusicParty
  (params (FadeOutDuration :type Float%20%28double-precision%29))
  (Variables|Default|SetbMusicPartyIsActive false)
  (Default|StopRadio))

(event Custom|Fade%20In%20Media
  (bind (_direction _vol) (Timeline|Volume%20Fader PlayFromStart)
    (:Update
      (Audio|Components|Audio|SetVolumeMultiplier (Variables|Default|GetMediaSound) _vol :!class /Script/AudioMixer.SynthComponent))))

(fn ChangeStation
  (params (Station%20Index :type Integer)))

(fn StartRadio
  (if (and (not (Variables|Default|GetbRadioIsActive)) (Variables|Default|GetbMusicPartyIsActive))
    (Default|DebounceCheck "StartRadio"
      (:Ready
        (Variables|Default|SetbRadioIsActive true)
        (Synth|Components|Audio|Start (Variables|Default|GetMediaSound))
        (CallFunction|Start)
        (Media|MediaPlayer|BindEventtoOnPlaybackResumed (Media|MediaSoundComponent|GetMediaPlayer (Variables|Default|GetMediaSound)) (EventDispatchers|CreateEvent :!event Fade%20In%20Media))))))

(fn StopRadio
  (if (and (Variables|Default|GetbRadioIsActive) (not (Variables|Default|GetbMusicPartyIsActive)))
    (Default|DebounceCheck "StopRadio"
      (:Ready
        (Variables|Default|SetbRadioIsActive false)
        (CallFunction|Stop)))))

(fn ConstructionScript)
