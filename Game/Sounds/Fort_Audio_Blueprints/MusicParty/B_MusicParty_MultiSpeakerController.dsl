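; Music Party controller Blueprint that reacts to events from its MusicPartyPropComponent as speakers in a multi-speaker setup are lost.
; Removing a speaker triggers an audio glitch: the simple delay preset briefly changes while a timeline animates the foldback distortion threshold, keeping its input and output gains.
; Losing all speakers stops the party, then destroys the actor after a short delay.
; The speaker-damaged handler is empty.
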(event OnSpeakerRemoved(MusicPartyPropComponent)
  (params (Speaker%20Actor :type Actor%20Object%20Reference))
  (CallFunction|TriggerFXMod))

(event OnSpeakerDamaged(MusicPartyPropComponent)
  (params (Speaker%20Actor :type Actor%20Object%20Reference)))

(event OnAllSpeakersDestroyed(MusicPartyPropComponent)
  (params (Speaker%20Actor :type Actor%20Object%20Reference))
  (Class|BaseClass:MusicPartyProp_Base|StopParty (Variables|Default|GetMusicPartyPropComponent) 0.25)
  (Utilities|FlowControl|Delay 3.0)
  (Actor|DestroyActor))

(event Custom|Trigger%20FX%20Mod
  (bind (_inputgaindb _thresholddb _outputgaindb) (Utilities|Struct|BreakSourceEffectFoldbackDistortionSettings (Variables|/Script/Synthesis.SourceEffectFoldbackDistortionPreset|GetSettings (Variables|Default|GetDistFX))))
  (Utilities|FlowControl|Sequence
    (:then_0
      (Audio|Effects|SetSettings (Variables|Default|GetDelayFX) (Utilities|Struct|MakeSourceEffectSimpleDelaySettings 343.0 0.2 0.0 1.0 0.0 false) :!class /Script/Synthesis.SourceEffectSimpleDelayPreset)
      (Utilities|FlowControl|Delay 0.1)
      (Audio|Effects|SetSettings (Variables|Default|GetDelayFX) (Utilities|Struct|MakeSourceEffectSimpleDelaySettings 343.0 0.1 0.0 1.0 0.0 false) :!class /Script/Synthesis.SourceEffectSimpleDelayPreset))
    (:then_1
      (bind (_direction _threshold) (Timeline|Timeline PlayFromStart)
        (:Update
          (Audio|Effects|SetSettings (Variables|Default|GetDistFX) (Utilities|Struct|MakeSourceEffectFoldbackDistortionSettings _inputgaindb _threshold _outputgaindb) :!class /Script/Synthesis.SourceEffectFoldbackDistortionPreset))))))

(fn ConstructionScript)
