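# Stereo MetaSound source for a stream-safe Music Party VUID track, blending a wave player with audio from an audio bus.
# A basic stereo wave player patch plays the input Waves from the start time, alongside a reader on MusicParty_VUID_AudioBus.
# The CB_LicensedAudio_Mute modulator value sets the player's mixer gain and its inverse sets the bus gain, crossfading between them.
# Base submix and submix sends are disabled, and virtualization uses seek restart.
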MetaSoundSource /Game/Sounds/Fort_Audio_Blueprints/MusicParty/MSS_MusicParty_StreamSafe_VUID_Track
inputs [
 UE.Source.OnPlay {type=Trigger}
 UE.Source.StartTime.StartTime {type=Time}
 Waves {type=WaveAsset:Array}
]
interfaces [UE.OutputFormat.Stereo:1.0, UE.Source.StartTime:1.0, UE.Source:1.1]
outputs [
 "UE.OutputFormat.Stereo.Audio:0" {type=Audio}
 "UE.OutputFormat.Stereo.Audio:1" {type=Audio}
]
pages [Default {
  input_defaults {UE.Source.OnPlay=null, UE.Source.StartTime.StartTime=null, Waves []}
  nodes [
   AudioBusReader2_0 = "UE.Audio Bus Reader (2)"() {also {defaults {Audio Bus=&/MusicParty_VUID_AudioBus}}}
   n1 = /FortMetasound/Patches/MSP_Audio_WavePlayer_Basic_Stereo(Play Trigger=^UE.Source.OnPlay, Start Time=^UE.Source.StartTime.StartTime, Wave Asset=^Waves) {also {defaults {Enable Shared Random State="<MetaSoundBuilderResult.SUCCEEDED: 0>", Is Looping?="<MetaSoundBuilderResult.SUCCEEDED: 0>"}}}
   n2 = Modulation.GetModulatorValue() {also {defaults {Modulator=/Game/Sounds/Modulation/Parameters/CB_LicensedAudio_Mute}}}
   AudioMixerAudioMixerStereo2_0 = "AudioMixer.Audio Mixer (Stereo, 2)"(
    In 0 L=@AudioBusReader2_0.Out 0
    In 0 R=@AudioBusReader2_0.Out 1
    In 1 L=@n1.Out L
    In 1 R=@n1.Out R
    Gain 1=@n2.Out
    Gain 0=UE.Subtract:Float(AdditionalOperands=@n2.Out) {also {defaults {PrimaryOperand=1}}}.Out
   )
  ]
  links [
   {dst="@UE.OutputFormat.Stereo.Audio:0", src=AudioMixerAudioMixerStereo2_0.Out L}
   {dst="@UE.OutputFormat.Stereo.Audio:1", src=AudioMixerAudioMixerStereo2_0.Out R}
  ]
 }]
properties {
 assetImportData=&self:AssetImportData
 bDebug=true
 bEnableBaseSubmix=false
 bEnableSubmixSends=false
 bLooping=true
 duration=10000
 numChannels=2
 outputFormat=Stereo
 sampleRate=48000
 virtualizationMode=SeekRestart
}
