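# Stereo MetaSound source for Music Party tracks, playing a randomly picked wave from the Waves input array.
# On Play triggers a Random Get, which passes the chosen wave and play trigger to a stereo Wave Player.
# The player begins at the StartTime interface value, allowing offset starts, and feeds both outputs directly.
# Waves is empty by default, and the source has no OneShot finish output.
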MetaSoundSource /Game/Sounds/Fort_Audio_Blueprints/MusicParty/MSS_MusicParty_Track
inputs [
 UE.Source.OnPlay {type=Trigger}
 UE.Source.StartTime.StartTime {type=Time}
 Waves {type=WaveAsset:Array}
]
interfaces [UE.OutputFormat.Stereo:1.0, UE.Source.StartTime:1.0, UE.Source:1.1]
outputs [
 "UE.OutputFormat.Stereo.Audio:0" {type=Audio}
 "UE.OutputFormat.Stereo.Audio:1" {type=Audio}
]
pages [Default {
  input_defaults {UE.Source.OnPlay=null, UE.Source.StartTime.StartTime=0, Waves []}
  nodes [
   ArrayRandomGet_0 = Array.Random Get:WaveAsset:Array(Next=^UE.Source.OnPlay, In Array=^Waves)
   WavePlayer_0 = UE.Wave Player:Stereo(Start Time=^UE.Source.StartTime.StartTime, Play=@ArrayRandomGet_0.On Next, Wave Asset=@ArrayRandomGet_0.Value) {also {defaults {Loop="<MetaSoundBuilderResult.SUCCEEDED: 0>"}, version [1,1]}}
  ]
  links [{dst="@UE.OutputFormat.Stereo.Audio:0", src=WavePlayer_0.Out Left}, {dst="@UE.OutputFormat.Stereo.Audio:1", src=WavePlayer_0.Out Right}]
 }]
properties {
 assetImportData=&self:AssetImportData
 bEnableBaseSubmix=false
 bEnableSubmixSends=false
 bLooping=true
 duration=10000
 numChannels=2
 outputFormat=Stereo
 sampleRate=48000
}
