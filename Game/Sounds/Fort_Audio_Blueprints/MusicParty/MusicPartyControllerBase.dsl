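; Base Blueprint controller for a Music Party that starts and stops the party's music.
; The construction script assigns the MusicToPlay sound to the Audio component when one is set.
; Start and stop events call the parent, then only on non-authority instances fade valid audio in from MusicStartTime or out over the supplied duration.
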(event EventStartMusicParty
  (|CallParentFunction)
  (Networking|SwitchHasAuthority
    (:Remote
      (bind _audio (Variables|Default|GetAudio)
        (:then
          (Audio|Components|Audio|FadeIn _audio 1.0 1.0 (Variables|Default|GetMusicStartTime) "Linear" :!class /Script/Engine.AudioComponent))
        (:else)))))

(event EventStopMusicParty
  (params (FadeOutDuration :type Float%20%28double-precision%29))
  (|CallParentFunction FadeOutDuration)
  (Networking|SwitchHasAuthority
    (:Remote
      (bind _audio (Variables|Default|GetAudio)
        (:then
          (Audio|Components|Audio|FadeOut _audio FadeOutDuration 0.0 "Linear" :!class /Script/Engine.AudioComponent))
        (:else)))))

(fn ConstructionScript
  (|CallParentFunction)
  (bind _musictoplay (Variables|Audio|GetMusicToPlay)
    (:then
      (Audio|Components|Audio|SetSound (Variables|Default|GetAudio) _musictoplay))
    (:else)))
