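; Blueprint interface for controlling a music party, part of the Fortnite audio blueprints.
; It declares StartMusicParty, which takes no parameters, and StopMusicParty, which takes a FadeOutDuration float.
; The file holds only these signatures, so the behavior is left to whatever implements the interface.
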(fn StartMusicParty)

(fn StopMusicParty
  (params (FadeOutDuration :type Float%20%28double-precision%29)))
