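; Blueprint interface for music party controllers in Fortnite audio, used to pass trigger parameters.
; It declares a single function, SetTriggerParameter, which takes a parameter name.
; The file holds only the signature, so implementing classes decide how the named trigger is applied.
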(fn SetTriggerParameter
  (params (ParameterName :type Name)))
