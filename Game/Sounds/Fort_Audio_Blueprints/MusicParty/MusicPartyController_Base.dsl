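; MusicPartyController_Base is a base Blueprint for a music party audio controller that tracks playback timing across starts and stops.
; Starting stamps game time and, if under two minutes have passed since the last stop, sets the music start time to that gap plus the time previously played.
; Stopping stamps game time and stores time played since the last start; its fade-out duration parameter is unused here.
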(event EventStartMusicParty
  (bind _returnvalue (Utilities|Time|GetGameTimeinSeconds))
  (bind _stopmusicpartytime (Variables|Audio|GetStopMusicPartyTime))
  (bind _returnvalue_2 (- _returnvalue _stopmusicpartytime))
  (Variables|Audio|SetStartMusicPartyTime _returnvalue)
  (if (< _returnvalue_2 1.2e+02)
    (Variables|Default|SetMusicStartTime (+ _returnvalue_2 (Variables|Audio|GetMusicPartyTimePlayed)))))

(event EventStopMusicParty
  (params (FadeOutDuration :type Float%20%28double-precision%29))
  (bind _output_get (Variables|Audio|SetStopMusicPartyTime (Utilities|Time|GetGameTimeinSeconds)))
  (Variables|Audio|SetMusicPartyTimePlayed (- _output_get (Variables|Audio|GetStartMusicPartyTime))))

(fn ConstructionScript)
