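; MusicPartyController_Radio_Fallback is a Blueprint that starts and stops radio music playback for Music Party, only on non-authority clients.
; StartRadio fades in its audio component, feeds the MusicToPlay array into the "Waves" parameter, and sends post-effect audio to a configured audio bus.
; StopRadio fades the audio out over a few seconds.
; Both overrides call their parent first and skip local audio work on the authority.
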(fn StartRadio
  (|CallParentFunction)
  (Networking|SwitchHasAuthority
    (:Remote
      (bind _audio (Variables|Default|GetAudio)
        (:then
          (Audio|Components|Audio|FadeIn _audio 1.0 1.0 0.0 "Linear" :!class /Script/Engine.AudioComponent)
          (Audio|Parameter|SetObjectArrayParameter _audio "Waves" (Variables|Default|GetMusicToPlay))
          (Audio|Components|Audio|SetAudioBusSendPostEffect (Variables|Default|GetAudio) (Variables|Default|GetAudio%20Bus) 1.0 :!class /Script/Engine.AudioComponent))
        (:else)))))

(fn StopRadio
  (|CallParentFunction)
  (Networking|SwitchHasAuthority
    (:Remote
      (bind _audio (Variables|Default|GetAudio)
        (:then
          (Audio|Components|Audio|FadeOut _audio 3.0 0.0 "Linear" :!class /Script/Engine.AudioComponent))
        (:else)))))

(fn ConstructionScript
  (|CallParentFunction))
