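; Music party controller Blueprint that plays the party's music as sound waves on instances without authority.
; On start, after the parent call, it fades in the Audio component from the music start time, passes the music to its "Waves" object array parameter and sets a post-effect send to an audio bus.
; On stop it fades out over the given duration.
; A helper converts the decibel adjustment to linear volume, with silence below a floor.
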(event Collision|EventActorBeginOverlap
  (params (OtherActor :type Actor%20Object%20Reference)))

(event EventTick
  (params (DeltaSeconds :type Float%20%28single-precision%29)))

(event EventStartMusicParty
  (bind _audio (Variables|Default|GetAudio))
  (|CallParentFunction)
  (Networking|SwitchHasAuthority
    (:Remote
      (Audio|Components|Audio|FadeIn (Variables|Default|GetAudio) 1.0 (Default|DecibelstoLinear (Variables|Audio|GetDbAdjustment)) (Variables|Default|GetMusicStartTime) "Linear" :!class /Script/Engine.AudioComponent)
      (Audio|Parameter|SetObjectArrayParameter _audio "Waves" (Variables|Audio|GetMusicToPlay))
      (Audio|Components|Audio|SetAudioBusSendPostEffect _audio (Variables|Audio|GetAudioBus) 1.0 :!class /Script/Engine.AudioComponent))))

(event EventStopMusicParty
  (params (FadeOutDuration :type Float%20%28double-precision%29))
  (|CallParentFunction FadeOutDuration)
  (Networking|SwitchHasAuthority
    (:Remote
      (bind _audio (Variables|Default|GetAudio)
        (:then
          (Audio|Components|Audio|FadeOut _audio FadeOutDuration 0.0 "Linear" :!class /Script/Engine.AudioComponent))
        (:else)))))

(fn DecibelsToLinear
  (params (Decibels :type Float%20%28double-precision%29))
  (if (>= Decibels -6e+01)
    (return (Math|Float|Power 1e+01 (/ Decibels 2e+01)))
    (else
      (return 0.0))))

(fn ConstructionScript)
