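; MusicPartyFunctionLibrary is a Blueprint function library for the MusicParty audio Blueprints, providing a single logging helper, PrintLog.
; PrintLog takes a Text message, formats it behind a "LogMusicParty:" prefix and converts it to a string.
; The result goes to FortShippingLog, Fortnite's shipping log node.
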(fn PrintLog
  (params (Log :type Text))
  (Fortnite|FortShippingLog 0 (Utilities|String|ToString(Text) (Utilities|Text|FormatText "NSLOCTEXT(\"[5CFA5389620435EDE2140059518DFB67]\", \"4BED561C4DE907080C6058A7BEF03F61\", \"LogMusicParty: {Log}\")" Log))))
