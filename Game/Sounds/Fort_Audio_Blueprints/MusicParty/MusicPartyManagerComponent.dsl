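; Manager component for music party prop speakers that starts and stops one shared music controller actor per controller class.
; Without cosmetic events it spawns missing controllers; with them it tracks speakers and refreshes playback when the controller list replicates.
; Audibility changes update per-class ref counts in a map, starting or stopping controllers through an interface.
; A global audible count registers or unregisters the default submix with audio analysis.
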(event EventEndPlay
  (params (EndPlayReason :type EEndPlayReason%20Enum :default "Destroyed"))
  (Utilities|FlowControl|Sequence
    (:then_0
      (for _array_element (Variables|Default|GetControllerActors)
        (Actor|DestroyActor _array_element)))
    (:then_1
      (if (> (Variables|Default|GetAudibleRefCount) 0)
        (AudioAnalysis|UnRegisterDefaultSubmix (WorldSubsystems|GetFortAudioAnalysisSubsystem))))))

(fn AddRef
  (params (Class :type Actor%20Class%20Reference%20%28by%20ref%29))
  (bind _speakercontrollermap (Variables|Default|GetSpeakerControllerMap))
  (bind (_value _returnvalue) (Utilities|Map|Find _speakercontrollermap Class))
  (if _returnvalue
    (bind _structout (Utilities|Struct|SetmembersinMusicPartyManagerControllerData _value (+ (Utilities|Struct|BreakMusicPartyManagerControllerData _value) 1)))
    (Utilities|Map|Add (Variables|Default|GetSpeakerControllerMap) (Variables|GetClass) _structout)
    (return true)
    (else
      (Utilities|Map|Add (Variables|Default|GetSpeakerControllerMap) Class (Utilities|Struct|MakeMusicPartyManagerControllerData 1))
      (return false))))

(fn GetControllerByClass
  (params (Class :type Object%20Class%20Reference))
  (for _array_element (Variables|Default|GetControllerActors)
    (if (== (Utilities|GetClass _array_element) Class)
      (return _array_element))))

(fn GetControllerData
  (params (Class :type Actor%20Class%20Reference%20%28by%20ref%29))
  (bind _speakercontrollermap (Variables|Default|GetSpeakerControllerMap))
  (bind (_value _returnvalue) (Utilities|Map|Find _speakercontrollermap Class))
  (bind (_value_speakerrefcount_2_d5b58ce54384d00665692095487f93d4 _value_bisplaying_6_f9944cc34a7fbc0b0e0cc29f1c03fb98) (Utilities|Struct|BreakMusicPartyManagerControllerData _value))
  (if _returnvalue
    (return true _value_speakerrefcount_2_d5b58ce54384d00665692095487f93d4 _value_bisplaying_6_f9944cc34a7fbc0b0e0cc29f1c03fb98)
    (else
      (return false 0 false))))

(fn OnRep_ControllerActors
  (Networking|CanExecuteCosmeticEvents
    (:True
      (SpeakerControllerMap|UpdatePlaying))))

(fn OnSpeakerIsAudible
  (params (Component :type Base%20Class:%20Music%20Party%20Prop%20Base%20Object%20Reference) (Is%20Audible :type Boolean))
  (bind _returnvalue (WorldSubsystems|GetFortAudioAnalysisSubsystem))
  (bind _audiblerefcount (Variables|Default|GetAudibleRefCount))
  (Utilities|FlowControl|Sequence
    (:then_0
      ;[0 0.15 0.04691 0.5] Check our Audible Count.  If we have one ref after adding, we need to register with the analysis system.  If we have zero refs after removing, we should unregister.
      (if Is%20Audible
        (bind _result (Math|Integer|IncrementInt _audiblerefcount)
          (:%20%20
            (if (== _result 1)
              (AudioAnalysis|RegisterDefaultSubmix _returnvalue))))
        (else
          (bind _result_2 (Math|Integer|DecrementInt _audiblerefcount)
            (:%20%20
              (if (== _result_2 0)
                (AudioAnalysis|UnRegisterDefaultSubmix _returnvalue)))))))
    (:then_1
      (SpeakerControllerMap|UpdateRefs (Initialize|GetControllerClass Component) Is%20Audible)
      (SpeakerControllerMap|UpdatePlaying))))

(fn OnSpeakersChanged
  (Variables|Default|SetbIsPartyActive (Utilities|Array|IsNotEmpty (Variables|Default|GetPartySpeakers))))

(fn RegisterControllerClass
  (params (Controller%20Class :type Actor%20Class%20Reference%20%28by%20ref%29))
  (Networking|CanExecuteCosmeticEvents
    (:False
      (Utilities|IsValid (CallFunction|GetControllerbyClass Controller%20Class)
        (:Is%20Not%20Valid
          (if (Utilities|IsValidClass Controller%20Class)
            (bind _returnvalue (Game|SpawnActorfromClass Controller%20Class (Math|Transform|MakeTransform "0, 0, 0") "Undefined" "SelectDefaultAtRuntime" 0 0))
            (Utilities|Array|AddUnique (Variables|Default|GetControllerActors) _returnvalue)))))))

(fn RegisterSpeaker
  (params (Component :type Base%20Class:%20Music%20Party%20Prop%20Base%20Object%20Reference%20%28by%20ref%29))
  (Networking|CanExecuteCosmeticEvents
    (:True
      (Utilities|Array|AddUnique (Variables|Default|GetPartySpeakers) (Variables|GetComponent)))))

(fn RemoveRef
  (params (Class :type Actor%20Class%20Reference%20%28by%20ref%29))
  (bind _speakercontrollermap (Variables|Default|GetSpeakerControllerMap))
  (bind (_value _returnvalue) (Utilities|Map|Find _speakercontrollermap Class))
  (bind _speakerrefcount_2_d5b58ce54384d00665692095487f93d4 (Utilities|Struct|BreakMusicPartyManagerControllerData _value))
  (bind _returnvalue_2 (+ _speakerrefcount_2_d5b58ce54384d00665692095487f93d4 -1))
  (if _returnvalue
    (bind _structout (Utilities|Struct|SetmembersinMusicPartyManagerControllerData _value _returnvalue_2))
    (Utilities|Map|Add (Variables|Default|GetSpeakerControllerMap) (Variables|GetClass) _structout)
    (return (== _returnvalue_2 0) true)
    (else
      (return false false))))

(fn SetIsPlaying
  (params (Class :type Actor%20Class%20Reference%20%28by%20ref%29) (bIsPlaying :type Boolean))
  (bind _class (Variables|GetClass))
  (bind _speakercontrollermap (Variables|Default|GetSpeakerControllerMap))
  (bind (_value _returnvalue) (Utilities|Map|Find _speakercontrollermap _class))
  (if _returnvalue
    (bind _structout (Utilities|Struct|SetmembersinMusicPartyManagerControllerData _value bIsPlaying))
    (Utilities|Map|Add (Variables|Default|GetSpeakerControllerMap) (Variables|GetClass) _structout)
    (return true)
    (else
      (return false))))

(fn UnregisterSpeaker
  (params (Component :type Base%20Class:%20Music%20Party%20Prop%20Base%20Object%20Reference%20%28by%20ref%29))
  (Networking|CanExecuteCosmeticEvents
    (:True
      (Utilities|Array|RemoveItem (Variables|Default|GetPartySpeakers) Component))))

(fn UpdatePlaying
  (for _array_element (Variables|Default|GetControllerActors)
    (Utilities|IsValid _array_element
      (:Is%20Valid
        (if (.Is%20Playing (SpeakerControllerMap|GetControllerData (Utilities|GetClass _array_element)))
          (if (== (.Speaker%20Ref%20Count (SpeakerControllerMap|GetControllerData (Utilities|GetClass _array_element))) 0)
            (Class|MusicPartyControllerInterface|StopMusicParty(Message) _array_element 0.5)
            (SpeakerControllerMap|SetIsPlaying (Utilities|GetClass _array_element)))
          (elif (> (.Speaker%20Ref%20Count (SpeakerControllerMap|GetControllerData (Utilities|GetClass _array_element))) 0)
            (Class|MusicPartyControllerInterface|StartMusicParty(Message) _array_element)
            (SpeakerControllerMap|SetIsPlaying (Utilities|GetClass _array_element) true)))))))

(fn UpdateRefs
  (params (Controller%20Class :type Actor%20Class%20Reference%20%28by%20ref%29) (Add%20Ref :type Boolean))
  (if Add%20Ref
    (SpeakerControllerMap|AddRef Controller%20Class)
    (else
      (SpeakerControllerMap|RemoveRef Controller%20Class))))
