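; Blueprint music party prop component that supplies sound assets and behavior settings for music played through speaker actors.
; Its setup override calls the parent, then passes its variables to SetupRuntimeVariables, including event tag and group, speaker damage and destruction sounds, controller class, and auto-update flags.
; CopySelfToNewOwner gets a new speaker array from the parent, adds this component class to the new owner with copied settings, and returns that array.
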(fn CallSetupRuntimeVariables
  (|CallParentFunction)
  (Initialize|SetupRuntimeVariables (Variables|Behavior|GetMusicEventTag) (Variables|Assets|GetSoundToPlay) (Variables|Behavior|GetMusicEventGroupName) (Variables|Assets|GetOnSpeakerDestroyedSound) (Variables|Assets|GetOnAllSpeakersDestroyedSound) (Variables|Assets|GetOnSpeakerDamagedSound) (Variables|Behavior|GetSpeakerActors) (Variables|Behavior|GetPlaybackStartTimeRange) (Variables|Behavior|GetAutoStopFadeOutTime) (Variables|Behavior|GetControllerClass) (Variables|Behavior|GetbAutoUpdateAudioSpread) (Variables|Behavior|GetbAutoUpdateAudioPosition)))

(fn CopySelfToNewOwner
  (params (NewOwner :type Actor%20Object%20Reference))
  (bind _newspeakeractorarray (|CallParentFunction NewOwner))
  (Game|AddComponentbyClass (Variables|GetNewOwner) "/Game/Sounds/Fort_Audio_Blueprints/MusicParty/MusicPartyPropComponent.MusicPartyPropComponent_C" (Variables|Assets|GetSoundToPlay) (Variables|Behavior|GetMusicEventTag) (Variables|Behavior|GetMusicEventGroupName) (Variables|Assets|GetOnSpeakerDamagedSound) (Variables|Assets|GetOnSpeakerDestroyedSound) (Variables|Assets|GetOnAllSpeakersDestroyedSound) (Variables|Behavior|GetControllerClass) _newspeakeractorarray (Variables|Behavior|GetbAutoUpdateAudioPosition) (Variables|Behavior|GetbAutoUpdateAudioSpread) (Variables|Assets|GetInstanceDbVolumeAdjustment))
  (return _newspeakeractorarray))
