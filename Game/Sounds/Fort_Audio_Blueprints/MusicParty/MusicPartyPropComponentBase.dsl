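; Base component for Music Party props that plays music through destructible speaker actors, with runtime settings supplied through an overridable setup hook.
; Once playlist data is ready it binds speaker events and the game state's manager component, then auto-starts audio, gated in Creative/UEFN.
; It registers with the manager and reports audibility from audio virtualization changes.
; Speaker loss triggers sounds, refits audio position and attenuation, and stops playback when none remain.
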(event EventBeginPlay
  ;[0 0 0] Begin Play
  (if (CallFunction|ShouldPlay)
    (bind _gamestate (Class|FortAsyncActionCurrentPlaylistReady|PlaylistDataReadyAsync)
      (:OnReady
        (if (not (Fortnite|IsSpatialGameplayExperience))
          _
          (elif (Variables|Private|GetbAllowInCreative)
            _))
        (Initialize|CallSetupRuntimeVariables)
        (Utilities|FlowControl|DoOnce false
          (:Completed
            (SpeakerActors|BindtoSpeakerDestruction)
            (CallFunction|SetupChildedSpeakerTracking)
            (Utilities|IsValid (Actor|GetComponentbyClass _gamestate "/Game/Sounds/Fort_Audio_Blueprints/MusicParty/MusicPartyManagerComponent.MusicPartyManagerComponent_C")
              (:Is%20Valid
                (Variables|Components|SetManagerComponent (Actor|GetComponentbyClass _gamestate "/Game/Sounds/Fort_Audio_Blueprints/MusicParty/MusicPartyManagerComponent.MusicPartyManagerComponent_C"))
                (CallFunction|OnManagerReady)))))))))

(event EventEndPlay
  (params (EndPlayReason :type EEndPlayReason%20Enum :default "Destroyed"))
  ;[0 0 0] End Play
  (Networking|CanExecuteCosmeticEvents
    (:True
      (CallFunction|StopParty (Variables|Runtime|GetRuntime_AutoStopFadeOutTime)))))

(event Custom|On%20Manager%20Ready
  (bind _runtime_trackclass (Variables|Runtime|GetRuntime_TrackClass))
  ;[0 0 0] On Manager Ready
  (Networking|CanExecuteCosmeticEvents
    (:True
      (CallFunction|SetupAutoActivate)
      (bind _managercomponent (Variables|Components|GetManagerComponent)
        (:then
          (Utilities|IsValid (Class|MusicPartyManagerComponent|GetControllerbyClass _managercomponent _runtime_trackclass)
            (:Is%20Valid
              (bind _asb_music_party_radio (Utilities|Casting|CastToB_MusicPartyRadio (Class|MusicPartyManagerComponent|GetControllerbyClass _managercomponent _runtime_trackclass))
                (:then
                  (EventDispatchers|BindEventtoOnSubmixChanged _asb_music_party_radio (EventDispatchers|CreateEvent :!event Custom|OnSubmixChanged))
                  (bind _returnvalue (Class|FortMediaStreamingRadio|IsLicensed _asb_music_party_radio))
                  (Variables|Default|SetIs%20Licensed _returnvalue))
                (:CastFailed
                  _))
              (if (Variables|Private|GetbShouldBePlaying)
                (CallFunction|SetSpeakerRegistered true)
                (CallFunction|UpdateAudibleRefCount)))
            (:Is%20Not%20Valid
              _)))
        (:else)))
    (:False
      (bind _managercomponent_2 (Variables|Components|GetManagerComponent)
        (:then
          (Default|RegisterControllerClass _managercomponent_2 (Initialize|GetControllerClass) :!class /Game/Sounds/Fort_Audio_Blueprints/MusicParty/MusicPartyManagerComponent.MusicPartyManagerComponent_C))
        (:else)))))

(event Custom|OnSubmixChanged
  (params (bIsLicensed :type Boolean))
  ;[0 0 0] On Manager Ready
  (Variables|Default|SetIs%20Licensed bIsLicensed))

(event Custom|Start%20Party
  (params (Audio%20Component :type Audio%20Component%20Object%20Reference) (FadeInDuration :type Float%20%28single-precision%29 :default 1.0))
  (bind (_runtime_playbackstarttimerange_x _runtime_playbackstarttimerange_y) (Math|Vector2D|BreakVector2D (Variables|GetRuntime_PlaybackStartTimeRange_X)))
  (bind _runtime_musiceventtag (Variables|Runtime|GetRuntime_MusicEventTag))
  ;[0 0 0] Party control (starts/stops source bus playback, registers/unregisters with manager)
  (if (CallFunction|CanStartParty)
    (Variables|Runtime|GetRuntime_SoundToPlay
      (:then
        (Networking|CanExecuteCosmeticEvents
          (:True
            (Variables|Private|SetbShouldBePlaying true)
            (Variables|Components|SetAudioComponent Audio%20Component)
            (Audio|Components|Audio|SetSound (Variables|Components|GetAudioComponent) (Variables|Runtime|GetRuntime_SoundToPlay))
            (CallFunction|SetSoundtoPlaySettings (Variables|Components|GetAudioComponent))
            (Default|UpdateAudioAttenuationAnLocation)
            (Audio|Components|Audio|FadeIn (Variables|Components|GetAudioComponent) FadeInDuration 1.0 (Math|Random|RandomFloatinRange _runtime_playbackstarttimerange_x _runtime_playbackstarttimerange_y) "Linear" :!class /Script/Engine.AudioComponent)
            (CallFunction|SetRealtimeSoundSettings (Variables|Components|GetAudioComponent))
            (if (GameplayTags|IsGameplayTagValid _runtime_musiceventtag)
              (Audio|AddEventbyComponent (GameInstanceSubsystems|GetMusicEventSubsystem) (Variables|Components|GetAudioComponent) _runtime_musiceventtag "(GameplayTags=)" 0 (Variables|Runtime|GetRuntime_MusicEventGroupName))
              (else
                _))
            (CallFunction|SetSpeakerRegistered true)
            (CallFunction|UpdateAudibleRefCount))))
      (:else))))

(event Custom|Stop%20Party
  (params (FadeOutDuration :type Float%20%28double-precision%29 :default 1.0))
  ;[0 0 0] Party control (starts/stops source bus playback, registers/unregisters with manager)
  (Networking|CanExecuteCosmeticEvents
    (:True
      (Variables|Private|SetbShouldBePlaying false)
      (bind _audiocomponent (Variables|Components|GetAudioComponent)
        (:then
          (Audio|Components|Audio|FadeOut _audiocomponent FadeOutDuration 0.0 "Linear" :!class /Script/Engine.AudioComponent)
          (EventDispatchers|UnbindEventfromOnAudioVirtualizationChanged _audiocomponent (EventDispatchers|CreateEvent :!event Custom|On%20Virtualization%20Changed))
          (if (GameplayTags|IsGameplayTagValid (Variables|Runtime|GetRuntime_MusicEventTag))
            (Audio|RemoveEvent (GameInstanceSubsystems|GetMusicEventSubsystem) (Variables|Components|GetAudioComponent))
            (else
              _))
          (CallFunction|SetSpeakerRegistered false)
          (Variables|Private|SetbIsAudible false)
          (CallFunction|UpdateAudibleRefCount))
        (:else)))))

(event Custom|SetupVirtualizationBindings
  (bind _audiocomponent (Variables|Components|GetAudioComponent))
  ;[0 0 0] Virtualization changed
  (EventDispatchers|BindEventtoOnAudioVirtualizationChanged _audiocomponent (EventDispatchers|CreateEvent :!event Custom|On%20Virtualization%20Changed))
  (CallFunction|OnVirtualizationChanged (Audio|Components|Audio|IsVirtualized _audiocomponent)))

(event Custom|On%20Speaker%20Damaged%20Rep
  (params (Location :type Vector))
  ;[0 0 0] Sound playback on Damage/Destruction
  (Networking|CanExecuteCosmeticEvents
    (:True
      (Audio|PlaySoundatLocation (Variables|Runtime|SpeakerData|GetRuntime_OnSpeakerDamagedSound) Location))))

(event Custom|On%20Speaker%20Removed%20Rep
  (params (Location :type Vector))
  ;[0 0 0] Sound playback on Damage/Destruction
  (Networking|CanExecuteCosmeticEvents
    (:True
      (Audio|PlaySoundatLocation (Variables|Runtime|SpeakerData|GetRuntime_OnSpeakerDestroyedSound) Location))))

(event Custom|On%20Virtualization%20Changed
  (params (bIsVirtualized :type Boolean))
  ;[0 0 0] Virtualization changed
  (Variables|Private|SetbIsAudible (not bIsVirtualized))
  (if (Variables|Private|GetbDelayOnVirtualized)
    (Utilities|Time|SetTimerbyEvent (delegate Custom|On%20Virtualization%20Changed%20Delay) 0.1)
    (else
      (CallFunction|OnVirtualizationChangedDelay))))

(event Custom|On%20All%20Speakers%20Removed%20Rep
  (params (Location :type Vector))
  ;[0 0 0] Sound playback on Damage/Destruction
  (Networking|CanExecuteCosmeticEvents
    (:True
      (Audio|PlaySoundatLocation (Variables|Runtime|SpeakerData|GetRuntime_OnAllSpeakersDestroyedSound) Location))))

(event Custom|On%20Virtualization%20Changed%20Delay
  (bind _runtime_musiceventtag (Variables|Runtime|GetRuntime_MusicEventTag))
  ;[0 0 0] Virtualization changed
  (Utilities|FlowControl|Sequence
    (:then_0
      (if (Variables|Private|GetbIsAudible)
        (bind _audiocomponent (Variables|Components|GetAudioComponent)
          (:then
            (if (GameplayTags|IsGameplayTagValid _runtime_musiceventtag)
              (Audio|AddEventbyComponent (GameInstanceSubsystems|GetMusicEventSubsystem) _audiocomponent _runtime_musiceventtag "(GameplayTags=)" 0 (Variables|Runtime|GetRuntime_MusicEventGroupName))))
          (:else))))
    (:then_1
      (CallFunction|UpdateAudibleRefCount))))

(fn BindToSpeakerDestruction
  (Networking|CanExecuteCosmeticEvents
    (:True
      ;[1 1 1] Client Events
      (SpeakerActors|ResolveSpeakerActors)
      (bind _asbuilding_actor (Utilities|Casting|CastToBuildingActor (Components|GetOwner))
        (:then
          (Utilities|Array|Add (Variables|Private|GetSpeakerActorsResolved) _asbuilding_actor))
        (:CastFailed
          _))
      (for _array_element (Variables|Private|GetSpeakerActorsResolved)
        (EventDispatchers|BindEventtoOnDeathPlayEffectsDelegate _array_element (EventDispatchers|CreateEvent :!event OnOwnerDeathEffectsClient))
        (Utilities|IsValid (Actor|GetComponentbyClass _array_element "/Game/Sounds/Fort_Audio_Blueprints/MusicParty/MusicPartyPropComponentBase.MusicPartyPropComponentBase_C")
          (:Is%20Valid
            (Health|BindEventtoOnBuildingHealthChanged _array_element (EventDispatchers|CreateEvent (Actor|GetComponentbyClass _array_element "/Game/Sounds/Fort_Audio_Blueprints/MusicParty/MusicPartyPropComponentBase.MusicPartyPropComponentBase_C") :!event OnBuildingHealthChangedClient)))
          (:Is%20Not%20Valid
            (Health|BindEventtoOnBuildingHealthChanged _array_element (EventDispatchers|CreateEvent :!event OnBuildingHealthChangedClient))))))))

(fn CallSetupRuntimeVariables)

(fn CanStartParty
  ;[1 1 1] FORT-689245 - Music Party Prop components placed in Creative/UEFN silence game events at a lower priorty. Allow opt-in for these modes.
  (LocalVariables|SetbInFNEPlaylist (or (Class|FortGameStateAthena|IsVKEditPlaylist (Fortnite|Athena|GameState|GetGameStateAthena)) (Class|FortGameStateAthena|IsVKPlayPlaylist (Fortnite|Athena|GameState|GetGameStateAthena))))
  (LocalVariables|SetbInCreativePlaylist (Class|FortGameStateAthena|IsCreativeorPlaygroundModePlaylist (Fortnite|Athena|GameState|GetGameStateAthena)))
  (if (LocalVariables|GetbInFNEPlaylist)
    (return (Variables|Private|GetbAllowInFNE))
    (elif (LocalVariables|GetbInCreativePlaylist)
      (return (Variables|Private|GetbAllowInCreative))
      (else
        (return true)))))

(fn CopySelfToNewOwner
  (params (NewOwner :type Actor%20Object%20Reference))
  (Utilities|IsValid (Actor|GetComponentbyClass NewOwner "/Script/Engine.AudioComponent")
    (:Is%20Valid
      _)
    (:Is%20Not%20Valid
      (bind _returnvalue (Game|AddComponentbyClass :self NewOwner :Class "/Script/Engine.AudioComponent"))
      (Transformation|SetWorldLocationAndRotation _returnvalue (Transformation|GetWorldLocation (Variables|Components|GetAudioComponent)) (Transformation|GetWorldRotation (Variables|Components|GetAudioComponent)) false false)))
  (for _array_element (Variables|Private|GetSpeakerActorsResolved)
    (Utilities|Array|AddUnique (LocalVariables|GetNewSpeakerArray) (Utilities|ResolveSoftReference _array_element)))
  (bind _audiocomponent (Variables|Components|GetAudioComponent)
    (:then
      (Audio|Components|Audio|FadeOut _audiocomponent 2.0 0.0 "Linear" :!class /Script/Engine.AudioComponent)
      (return (LocalVariables|GetNewSpeakerArray)))
    (:else)))

(fn DecibelsToLinear
  (params (Decibels :type Float%20%28double-precision%29))
  (if (>= Decibels -6e+01)
    (return (Math|Float|Power 1e+01 (/ Decibels 2e+01)))
    (else
      (return 0.0))))

(fn GetControllerClass
  (return (Variables|Runtime|GetRuntime_TrackClass)))

(fn IsOwnedByBuildingActor
  (bind (_asbuilding_actor _bsuccess) (Utilities|Casting|CastToBuildingActor (Components|GetOwner)))
  (return _bsuccess))

(fn OnBuildingHealthChangedClient
  (CallFunction|OnSpeakerDamagedRep (Transformation|GetWorldLocation (select (Utilities|IsValid (Variables|Components|GetAudioComponent) :!class /Script/Engine.KismetSystemLibrary) (Variables|Components|GetAudioComponent) (Variables|/Script/Engine.Actor|GetRootComponent (Components|GetOwner self))))))

(fn OnOwnerDeathEffectsClient
  (params (Damage :type Float%20%28single-precision%29) (DamageTags :type Gameplay%20Tag%20Container%20Structure%20%28by%20ref%29) (Momentum :type Vector) (HitInfo :type Hit%20Result%20Structure%20%28by%20ref%29) (InstigatedBy :type Fort%20Pawn%20Object%20Reference) (DamageCauser :type Actor%20Object%20Reference) (EffectContext :type Gameplay%20Effect%20Context%20Handle%20Structure))
  (bind (_bblockinghit _binitialoverlap _time _distance _location _impactpoint _normal _impactnormal _physmat _hitactor) (Collision|BreakHitResult HitInfo))
  (bind _speakeractorsresolved (Variables|Private|GetSpeakerActorsResolved))
  (bind _output (Utilities|Array|Get(acopy) _speakeractorsresolved))
  (LocalVariables|SetIsOwnerDeath (== _hitactor (Components|GetOwner)))
  (bind _returnvalue (Utilities|Array|RemoveItem (Variables|Private|GetSpeakerActorsResolved) (Utilities|Casting|CastToBuildingActor _hitactor)))
  (if _returnvalue
    (Default|UpdateAudioAttenuationAnLocation)
    (if (Utilities|Array|IsEmpty (Variables|Private|GetSpeakerActorsResolved))
      (CallFunction|OnAllSpeakersRemovedRep "0, 0, 0")
      (CallFunction|StopParty (Variables|Runtime|GetRuntime_AutoStopFadeOutTime))
      (elif (LocalVariables|GetIsOwnerDeath)
        (CallFunction|OnSpeakerRemovedRep (Transformation|GetActorLocation (Components|GetOwner)))
        (Utilities|IsValid _output
          (:Is%20Valid
            (Initialize|CopySelftoNewOwner _output)))))))

(fn OnRep_SpeakerActorsResolved
  (if (Utilities|Array|IsNotEmpty (Variables|Runtime|GetRuntime_SpeakerActors))
    (if (Utilities|Array|IsEmpty (Variables|Private|GetSpeakerActorsResolved))
      (Delegates|CallOnAllSpeakersDestroyed (Components|GetOwner))
      (CallFunction|StopParty (Variables|Runtime|GetRuntime_AutoStopFadeOutTime))
      (else
        (Default|UpdateAudioAttenuationAnLocation)))))

(fn RegisterControllerClass
  (bind _managercomponent (Variables|Components|GetManagerComponent)
    (:then
      (Default|RegisterControllerClass _managercomponent (Initialize|GetControllerClass) :!class /Game/Sounds/Fort_Audio_Blueprints/MusicParty/MusicPartyManagerComponent.MusicPartyManagerComponent_C))
    (:else)))

(fn RemoveSpeakerServer
  (params (Speaker :type Building%20Actor%20Object%20Reference%20%28by%20ref%29))
  (bind _returnvalue (Transformation|GetActorLocation Speaker))
  (bind _returnvalue_2 (Utilities|Array|RemoveItem (Variables|Private|GetSpeakerActorsResolved) Speaker))
  (if _returnvalue_2
    (CallFunction|OnSpeakerRemovedRep _returnvalue)
    (Delegates|CallOnSpeakerRemoved (Components|GetOwner))
    (if (Utilities|Array|IsEmpty (Variables|Private|GetSpeakerActorsResolved))
      (Delegates|CallOnAllSpeakersDestroyed (Components|GetOwner))
      (CallFunction|OnAllSpeakersRemovedRep _returnvalue))))

(fn ResolveSpeakerActors
  (for _array_element (Variables|Runtime|GetRuntime_SpeakerActors)
    (Utilities|Array|Add (LocalVariables|GetResolved) (Utilities|ResolveSoftReference _array_element)))
  (Variables|Private|SetSpeakerActorsResolved (LocalVariables|GetResolved)))

(fn SetRealtimeSoundSettings
  (params (AudioComponent :type Audio%20Component%20Object%20Reference))
  (CallFunction|SetupVirtualizationBindings))

(fn SetSoundToPlaySettings
  (params (AudioComponent :type Audio%20Component%20Object%20Reference))
  (Audio|Components|Audio|AdjustVolume AudioComponent 0.0 (Default|DecibelstoLinear (Variables|Assets|GetInstanceDbVolumeAdjustment)) "Linear" :!class /Script/Engine.AudioComponent))

(fn SetSpeakerRegistered
  (params (Registered :type Boolean))
  (bind _self self)
  (if (!= Registered (Variables|Private|GetbIsRegistered))
    (bind _managercomponent (Variables|Components|GetManagerComponent)
      (:then
        (bind _output_get (Variables|Private|SetbIsRegistered Registered))
        (if _output_get
          (Class|MusicPartyManagerComponent|RegisterSpeaker _managercomponent _self)
          (else
            (Class|MusicPartyManagerComponent|UnregisterSpeaker _managercomponent _self))))
      (:else))))

(fn SetupAutoActivate
  (bind _returnvalue (Components|GetOwner))
  (bind _returnvalue_2 (Actor|GetComponentbyClass _returnvalue "/Script/Engine.AudioComponent"))
  (if (Math|Boolean|ANDBoolean (Utilities|IsValid (Variables|Runtime|GetRuntime_SoundToPlay) :!class /Script/Engine.KismetSystemLibrary) (Variables|Activation|GetbAutoActivate) (not (Variables|Assets|GetBypassMusic)))
    (Utilities|IsValid _returnvalue_2
      (:Is%20Valid
        (CallFunction|StartParty _returnvalue_2)))))

(fn SetupChildedSpeakerTracking
  (if (Utilities|Array|IsNotEmpty (Variables|Default|GetChildSpeakerComponents))
    (for _array_element (Variables|Default|GetChildSpeakerComponents)
      (Class|MusicPartyChildSpeakerComponent|SetupSpeakerAssetVariables _array_element (Variables|Runtime|SpeakerData|GetRuntime_OnSpeakerDamagedSound) (Variables|Runtime|SpeakerData|GetRuntime_OnAllSpeakersDestroyedSound)))
    (return (Variables|Default|GetChildSpeakerComponents))
    (else
      (return (Variables|Default|GetChildSpeakerComponents)))))

(fn SetupRuntimeVariables
  (params (MusicEventTag :type Gameplay%20Tag%20Structure) (SoundToPlay :type Sound%20Base%20Object%20Reference) (MusicEventGroupName :type Name) (OnSpeakerDestroyed :type Sound%20Base%20Object%20Reference) (OnAllSpeakersDestroyed :type Sound%20Base%20Object%20Reference) (OnSpeakerDamaged :type Sound%20Base%20Object%20Reference) (SpeakerActors :type Array%20of%20Building%20Actor%20Soft%20Object%20References) (PlaybackStartTimeRandom :type Vector%202D%20Structure) (AutoStopFadeOutTime :type Float%20%28double-precision%29) (MusicTrackClass :type Actor%20Class%20Reference) (bAutoUpdateAudioSpread :type Boolean) (bAutoUpdateAudioPosition :type Boolean))
  (Variables|Runtime|SetRuntime_MusicEventTag MusicEventTag)
  (Variables|Runtime|SetRuntime_SoundToPlay SoundToPlay)
  (Variables|Runtime|SetRuntime_MusicEventGroupName MusicEventGroupName)
  (Variables|Runtime|SpeakerData|SetRuntime_OnSpeakerDestroyedSound (Variables|GetOnSpeakerDestroyed))
  (Variables|Runtime|SpeakerData|SetRuntime_OnAllSpeakersDestroyedSound (Variables|Delegates|GetOnAllSpeakersDestroyed))
  (Variables|Runtime|SpeakerData|SetRuntime_OnSpeakerDamagedSound (Variables|Delegates|GetOnSpeakerDamaged))
  (Variables|Runtime|SetRuntime_SpeakerActors (Variables|GetSpeakerActors))
  (Variables|Runtime|SetRuntime_PlaybackStartTimeRange (Variables|GetPlaybackStartTimeRandom))
  (Variables|Runtime|SetRuntime_AutoStopFadeOutTime (Variables|GetAutoStopFadeOutTime))
  (Variables|Runtime|SetRuntime_bAutoUpdateAudioPosition (Variables|GetbAutoUpdateAudioPosition))
  (Variables|Runtime|SetRuntime_bAutoUpdateAudioSpread (Variables|GetbAutoUpdateAudioSpread))
  (Variables|Runtime|SetRuntime_TrackClass (Variables|GetMusicTrackClass)))

(fn SetupSubmixSend
  (params (IsLicensed :type Boolean)))

(fn ShouldPlay
  (bind _perplatformplaystatus_default (Utilities|Struct|BreakPerPlatformBool (Variables|GetPerPlatformPlayStatus_Default)))
  (return (or _perplatformplaystatus_default (Variables|Assets|GetbPlayOnAllPlatforms))))

(fn UpdateAudibleRefCount
  (bind _self self)
  (bind _managercomponent (Variables|Components|GetManagerComponent))
  (Variables|Components|GetManagerComponent
    (:then
      (if (Variables|Private|GetbIsAudible)
        (if (not (Variables|Private|GetbIsAudibleRefCountRegistered))
          (Variables|Private|SetbIsAudibleRefCountRegistered true)
          (Class|MusicPartyManagerComponent|OnSpeakerIsAudible _managercomponent _self true))
        (elif (Variables|Private|GetbIsAudibleRefCountRegistered)
          (Variables|Private|SetbIsAudibleRefCountRegistered false)
          (Class|MusicPartyManagerComponent|OnSpeakerIsAudible _managercomponent _self false))))
    (:else)))

(fn UpdateAudioAttenuationAnLocation
  (bind _audiocomponent (Variables|Components|GetAudioComponent))
  (Variables|Components|GetAudioComponent
    (:then
      (Utilities|FlowControl|Sequence
        (:then_0
          (if (Variables|Runtime|GetRuntime_bAutoUpdateAudioPosition)
            (bind _center (Collision|GetActorArrayBounds (Variables|Private|GetSpeakerActorsResolved)))
            (Transformation|SetWorldLocation (Variables|Components|GetAudioComponent) _center false false)))
        (:then_1
          (if (Variables|Runtime|GetRuntime_bAutoUpdateAudioSpread)
            (bind _outattenuationsettings (Audio|Components|Audio|GetAttenuationSettingsToApply _audiocomponent))
            (Variables|/Script/Engine.AudioComponent|SetAttenuationOverrides _outattenuationsettings _audiocomponent)
            (Variables|/Script/Engine.AudioComponent|SetbOverrideAttenuation true _audiocomponent)
            (bind (_center_2 _boxextent) (Collision|GetActorArrayBounds (Variables|Private|GetSpeakerActorsResolved)))
            (bind _structout (Utilities|Struct|SetmembersinSoundAttenuationSettings (Variables|/Script/Engine.AudioComponent|GetAttenuationOverrides (Variables|Components|GetAudioComponent)) (* (Math|Float|Max(Float) (.x _boxextent) (.y _boxextent)) 2.0)))
            (Audio|Components|Audio|AdjustAttenuation (Variables|Components|GetAudioComponent) _structout)))))
    (:else)))
