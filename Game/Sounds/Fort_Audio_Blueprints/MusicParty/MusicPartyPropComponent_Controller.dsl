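; Blueprint function graph for a Music Party prop component controller, part of the game's audio Blueprints.
; It defines GetControllerClass, which returns the controller class stored in the asset's variable.
; The file holds only this getter, with no other logic.
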(fn GetControllerClass
  (return (Variables|Assets|GetControllerClass)))
