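; Music Party prop component variant that takes its music speaker setup from a MusicPartySpeaker asset.
; It feeds the asset's MetaSound, damage and destruction sounds into runtime setup and child speaker components.
; It applies the asset's attenuation, filters, volume offset, effect chain and concurrency to the audio component, with the audio bus coming from the track controller class or a radio default.
; It can copy itself onto a new owner actor.
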(event EventBeginPlay
  (bind _returnvalue (Components|GetOwner))
  (bind _returnvalue_2 (Actor|GetComponentbyClass _returnvalue "/Script/Engine.AudioComponent"))
  (|CallParentFunction)
  (Utilities|IsValid _returnvalue_2
    (:Is%20Valid
      (Variables|Initialize|SetRealtime_AudioComponent _returnvalue_2))))

(event EventStartParty
  (params (Audio%20Component :type Audio%20Component%20Object%20Reference) (FadeInDuration :type Float%20%28single-precision%29))
  (|CallParentFunction Audio%20Component FadeInDuration)
  (Variables|Initialize|SetRealtime_AudioComponent Audio%20Component))

(fn CallSetupRuntimeVariables
  (|CallParentFunction)
  (bind _speakerasset (Variables|Assets|GetSpeakerAsset)
    (:then
      (Initialize|SetupRuntimeVariables :MusicEventTag (Variables|Behavior|GetMusicEventTag) :SoundToPlay (Variables|/Game/Sounds/Fort_Audio_Blueprints/MusicParty/MusicPartySpeaker.MusicPartySpeaker_C|GetMetaSoundBase _speakerasset) :MusicEventGroupName (Variables|Behavior|GetMusicEventGroupName) :OnSpeakerDestroyed (Variables|/Game/Sounds/Fort_Audio_Blueprints/MusicParty/MusicPartySpeaker.MusicPartySpeaker_C|GetOnSpeakerDestroyedSound _speakerasset) :OnAllSpeakersDestroyed (Variables|/Game/Sounds/Fort_Audio_Blueprints/MusicParty/MusicPartySpeaker.MusicPartySpeaker_C|GetOnAllSpeakersDestroyedSound _speakerasset) :OnSpeakerDamaged (Variables|/Game/Sounds/Fort_Audio_Blueprints/MusicParty/MusicPartySpeaker.MusicPartySpeaker_C|GetOnSpeakerDamagedSound _speakerasset) :SpeakerActors (Variables|Behavior|GetSpeakerActors) :AutoStopFadeOutTime 0.5 :MusicTrackClass (Initialize|GetControllerClass) :bAutoUpdateAudioSpread (> (Utilities|Array|Length (Variables|Behavior|GetSpeakerActors)) 0) :bAutoUpdateAudioPosition (> (Utilities|Array|Length (Variables|Behavior|GetSpeakerActors)) 0)))
    (:else)))

(fn CopySelfToNewOwner
  (params (NewOwner :type Actor%20Object%20Reference))
  (bind _newspeakeractorarray (|CallParentFunction NewOwner))
  (Game|AddComponentbyClass (Variables|GetNewOwner) "/Game/Sounds/Fort_Audio_Blueprints/MusicParty/MusicPartyPropComponent_SpeakerAssets.MusicPartyPropComponent_SpeakerAssets_C" (Variables|Behavior|GetMusicEventTag) (Variables|Behavior|GetMusicEventGroupName) _newspeakeractorarray (Variables|Assets|GetSpeakerAsset) (Initialize|GetControllerClass) (Variables|Assets|GetInstanceDbVolumeAdjustment)))

(fn GetControllerClass
  (return (Variables|Initialize|GetCloned_Controller)))

(fn SetAudioBus
  (bind _realtime_audiocomponent (Variables|Initialize|GetRealtime_AudioComponent))
  (bind _asmusic_party_controller_base (Utilities|Casting|CastToMusicPartyController_BaseClass (Variables|Runtime|GetRuntime_TrackClass))
    (:then
      (Audio|Parameter|SetObjectParameter _realtime_audiocomponent "AudioBus" (Class|GetClassDefaults _asmusic_party_controller_base)))
    (:CastFailed
      (Audio|Parameter|SetObjectParameter _realtime_audiocomponent "AudioBus" (Variables|Other|GetRadioDefaultAudioBus)))))

(fn SetRealtimeSoundSettings
  (params (AudioComponent :type Audio%20Component%20Object%20Reference))
  (bind _asmusic_party_controller_base (Utilities|Casting|CastToMusicPartyController_BaseClass (Variables|Runtime|GetRuntime_TrackClass))
    (:then
      (Audio|Parameter|SetObjectParameter AudioComponent "AudioBus" (Class|GetClassDefaults _asmusic_party_controller_base)))
    (:CastFailed
      (Audio|Parameter|SetObjectParameter AudioComponent "AudioBus" (Variables|Other|GetRadioDefaultAudioBus))))
  (Audio|Parameter|SetBooleanParameter AudioComponent "Set" true :!class /Script/AudioExtensions.AudioParameterControllerInterface)
  (bind _speakerasset (Variables|Assets|GetSpeakerAsset)
    (:then
      (Utilities|IsValid (Variables|/Game/Sounds/Fort_Audio_Blueprints/MusicParty/MusicPartySpeaker.MusicPartySpeaker_C|GetSoundAttenuation _speakerasset)
        (:Is%20Valid
          (Variables|/Script/Engine.AudioComponent|SetAttenuationSettings (Variables|/Game/Sounds/Fort_Audio_Blueprints/MusicParty/MusicPartySpeaker.MusicPartySpeaker_C|GetSoundAttenuation _speakerasset) AudioComponent)
          (Audio|Parameter|SetFloatParameter AudioComponent "LPF" (Variables|/Game/Sounds/Fort_Audio_Blueprints/MusicParty/MusicPartySpeaker.MusicPartySpeaker_C|GetLPF (Variables|Assets|GetSpeakerAsset)) :!class /Script/AudioExtensions.AudioParameterControllerInterface)
          (Audio|Parameter|SetFloatParameter AudioComponent "HPF" (Variables|/Game/Sounds/Fort_Audio_Blueprints/MusicParty/MusicPartySpeaker.MusicPartySpeaker_C|GetHPF (Variables|Assets|GetSpeakerAsset)) :!class /Script/AudioExtensions.AudioParameterControllerInterface)
          (Audio|Components|Audio|AdjustVolume AudioComponent 0.0 (Default|DecibelstoLinear (+ (Variables|Assets|GetInstanceDbVolumeAdjustment) (Variables|/Game/Sounds/Fort_Audio_Blueprints/MusicParty/MusicPartySpeaker.MusicPartySpeaker_C|GetVolumeDbAdjustment (Variables|Assets|GetSpeakerAsset)))) "Linear" :!class /Script/Engine.AudioComponent)
          (|CallParentFunction AudioComponent))))
    (:else)))

(fn SetRealtimeSoundSettings_Deprecated
  (params (AudioComponent :type Audio%20Component%20Object%20Reference))
  (bind _audiocomponent (Variables|Components|GetAudioComponent))
  (bind _asmusic_party_controller_base (Utilities|Casting|CastToMusicPartyController_BaseClass (Variables|Runtime|GetRuntime_TrackClass))
    (:then
      (Audio|Parameter|SetObjectParameter (Variables|Components|GetAudioComponent) "AudioBus" (Class|GetClassDefaults _asmusic_party_controller_base)))
    (:CastFailed
      (Audio|Parameter|SetObjectParameter (Variables|Components|GetAudioComponent) "AudioBus" (Variables|Other|GetRadioDefaultAudioBus))))
  (Audio|Parameter|SetBooleanParameter (Variables|Components|GetAudioComponent) "Set" true :!class /Script/AudioExtensions.AudioParameterControllerInterface)
  (bind _speakerasset (Variables|Assets|GetSpeakerAsset)
    (:then
      (Utilities|IsValid (Variables|/Game/Sounds/Fort_Audio_Blueprints/MusicParty/MusicPartySpeaker.MusicPartySpeaker_C|GetSoundAttenuation _speakerasset)
        (:Is%20Valid
          (Variables|/Script/Engine.AudioComponent|SetAttenuationSettings (Variables|/Game/Sounds/Fort_Audio_Blueprints/MusicParty/MusicPartySpeaker.MusicPartySpeaker_C|GetSoundAttenuation _speakerasset) _audiocomponent)
          (Audio|Parameter|SetFloatParameter _audiocomponent "LPF" (Variables|/Game/Sounds/Fort_Audio_Blueprints/MusicParty/MusicPartySpeaker.MusicPartySpeaker_C|GetLPF (Variables|Assets|GetSpeakerAsset)) :!class /Script/AudioExtensions.AudioParameterControllerInterface)
          (Audio|Parameter|SetFloatParameter _audiocomponent "HPF" (Variables|/Game/Sounds/Fort_Audio_Blueprints/MusicParty/MusicPartySpeaker.MusicPartySpeaker_C|GetHPF (Variables|Assets|GetSpeakerAsset)) :!class /Script/AudioExtensions.AudioParameterControllerInterface)
          (Audio|Components|Audio|AdjustVolume _audiocomponent 0.0 (Default|DecibelstoLinear (+ (Variables|Assets|GetInstanceDbVolumeAdjustment) (Variables|/Game/Sounds/Fort_Audio_Blueprints/MusicParty/MusicPartySpeaker.MusicPartySpeaker_C|GetVolumeDbAdjustment (Variables|Assets|GetSpeakerAsset)))) "Linear" :!class /Script/Engine.AudioComponent))))
    (:else)))

(fn SetSoundToPlaySettings
  (params (AudioComponent :type Audio%20Component%20Object%20Reference))
  (bind _audiocomponent (Variables|Components|GetAudioComponent))
  (bind _speakerasset (Variables|Assets|GetSpeakerAsset))
  (bind _sourcepresetchain (Variables|/Game/Sounds/Fort_Audio_Blueprints/MusicParty/MusicPartySpeaker.MusicPartySpeaker_C|GetSourcePresetChain _speakerasset))
  (if (or (Utilities|IsValid _sourcepresetchain :!class /Script/Engine.KismetSystemLibrary) (> (Utilities|Set|Length (Variables|/Game/Sounds/Fort_Audio_Blueprints/MusicParty/MusicPartySpeaker.MusicPartySpeaker_C|GetConcurrencySet _speakerasset)) 0))
    (Audio|Components|Audio|Stop (Variables|Components|GetAudioComponent))
    (Variables|/Script/Engine.AudioComponent|SetSourceEffectChain _sourcepresetchain _audiocomponent)
    (Variables|/Script/Engine.AudioComponent|SetConcurrencySet (Variables|/Game/Sounds/Fort_Audio_Blueprints/MusicParty/MusicPartySpeaker.MusicPartySpeaker_C|GetConcurrencySet (Variables|Assets|GetSpeakerAsset)) _audiocomponent)
    (Audio|Components|Audio|Play (Variables|Components|GetAudioComponent) 0.0)
    (else
      _))
  (CallFunction|SetRealtimeSoundSettings (Variables|Components|GetAudioComponent)))

(fn SetupChildedSpeakerTracking
  (bind _childspeakercomponents (|CallParentFunction))
  (if (Utilities|Array|IsNotEmpty _childspeakercomponents)
    (bind _speakerasset (Variables|Assets|GetSpeakerAsset)
      (:then
        (for _array_element _childspeakercomponents
          (Class|MusicPartyChildSpeakerComponent|SetupSpeakerAssetVariables _array_element (Variables|/Game/Sounds/Fort_Audio_Blueprints/MusicParty/MusicPartySpeaker.MusicPartySpeaker_C|GetOnSpeakerDamagedSound _speakerasset) (Variables|/Game/Sounds/Fort_Audio_Blueprints/MusicParty/MusicPartySpeaker.MusicPartySpeaker_C|GetOnSpeakerDestroyedSound _speakerasset)))
        (return _childspeakercomponents))
      (:else))))
