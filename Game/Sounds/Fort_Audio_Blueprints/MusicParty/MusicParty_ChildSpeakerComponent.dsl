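; Music Party child speaker component that gives a speaker building actor damage and destruction audio and links it to its parent Music Party prop.
; On BeginPlay it casts its owner to BuildingActor and binds to the health-changed and death-effects delegates.
; Health changes play the damaged sound; death forwards effect data to a valid parent and plays the destroyed sound.
; Setup functions supply the parent and sounds; GetMusicPartyPropParent sets the parent despite its name.
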(event EventBeginPlay
  (bind _asbuilding_actor (Utilities|Casting|CastToBuildingActor (Components|GetOwner)))
  (Health|BindEventtoOnBuildingHealthChanged _asbuilding_actor (EventDispatchers|CreateEvent :!event OnBuildingHealthChanged))
  (EventDispatchers|BindEventtoOnDeathPlayEffectsDelegate _asbuilding_actor (EventDispatchers|CreateEvent :!event OnOwnerDeathEffectClient)))

(event EventTick
  (params (DeltaSeconds :type Float%20%28single-precision%29)))

(event Custom|Initialize
  (params (MusicPartyParent :type Base%20Class:%20Music%20Party%20Prop%20Base%20Object%20Reference))
  (Variables|Default|SetMusicPartyPropComponentParent MusicPartyParent))

(fn GetMusicPartyPropParent
  (params (MusicPartyPropComponentParent :type Base%20Class:%20Music%20Party%20Prop%20Base%20Object%20Reference))
  (Variables|Default|SetMusicPartyPropComponentParent MusicPartyPropComponentParent))

(fn OnBuildingHealthChanged
  (Audio|PlaySoundatLocation (Variables|Default|GetOn%20Speaker%20Damaged%20Sound) (Transformation|GetActorLocation (Components|GetOwner))))

(fn OnOwnerDeathEffectClient
  (params (Damage :type Float%20%28single-precision%29) (DamageTags :type Gameplay%20Tag%20Container%20Structure%20%28by%20ref%29) (Momentum :type Vector) (HitInfo :type Hit%20Result%20Structure%20%28by%20ref%29) (InstigatedBy :type Fort%20Pawn%20Object%20Reference) (DamageCauser :type Actor%20Object%20Reference) (EffectContext :type Gameplay%20Effect%20Context%20Handle%20Structure))
  (bind _musicpartypropcomponentparent (Variables|Default|GetMusicPartyPropComponentParent)
    (:then
      (SpeakerActors|Client|OnOwnerDeathEffectsClient _musicpartypropcomponentparent Damage DamageTags Momentum HitInfo InstigatedBy DamageCauser EffectContext))
    (:else
      _))
  (Audio|PlaySoundatLocation (Variables|Default|GetOn%20Speaker%20Destroyed%20Sound) (Transformation|GetActorLocation (Components|GetOwner)))
  (return))

(fn SetupSpeakerAssetVariables
  (params (DamageSound :type Sound%20Base%20Object%20Reference) (DestroySound :type Sound%20Base%20Object%20Reference))
  (Variables|Default|SetOn%20Speaker%20Damaged%20Sound DamageSound)
  (Variables|Default|SetOn%20Speaker%20Destroyed%20Sound DestroySound))
